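; Foraged health and shield regen item actor that a pawn interacts with to consume.
; Interaction is always permitted, while consumption requires a valid pawn with health or shield not nearly equal to its maximum, plus approval from the parent check.
; The regen effect itself is not in these graphs; the event stubs are disabled and the construction script adds a zero offset to the static mesh.
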(event EventBeginPlay
  (|CallParentFunction) ; This node is disabled and will not be called.\nDrag off pins to build functionality.
)

(event Collision|EventActorBeginOverlap
  (params (OtherActor :type Actor%20Object%20Reference))
  (|CallParentFunction OtherActor) ; This node is disabled and will not be called.\nDrag off pins to build functionality.
)

(event EventTick
  (params (DeltaSeconds :type Float%20%28single-precision%29))
  (|CallParentFunction DeltaSeconds) ; This node is disabled and will not be called.\nDrag off pins to build functionality.
)

(fn BlueprintCanInteract
  (params (InteractingPawn :type Fort%20Pawn%20Object%20Reference) (InteractionBeingAttempted :type EInteractionBeingAttempted%20Enum :default "FirstInteraction") (InteractionType :type TInteractionType%20Enum :default "IT_NoInteraction"))
  (return true))

(fn OnInteractAllowedToConsume
  (params (InteractingPawn :type Fort%20Pawn%20Object%20Reference) (InteractionBeingAttempted :type EInteractionBeingAttempted%20Enum :default "FirstInteraction"))
  (Utilities|IsValid InteractingPawn
    (:Is%20Valid
      (return (and (or (not (Math|Float|NearlyEqual(Float) (FortPawn|GetHealth InteractingPawn) (FortPawn|GetMaxHealth InteractingPawn))) (not (Math|Float|NearlyEqual(Float) (FortPawn|GetShield InteractingPawn) (FortPawn|GetMaxShield InteractingPawn)))) (|CallParentFunction InteractingPawn "FirstInteraction" "IT_Simple"))))
    (:Is%20Not%20Valid
      (return false))))

(fn ConstructionScript
  (|CallParentFunction)
  (Transformation|AddLocalOffset (Variables|Default|GetStaticMesh) "0, 0, 0" false false))
